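# Translucent, two-sided water trail material for Cattus event effects, flagged for Niagara ribbons, mesh particles, sprites and beam trails.
# Emissive blends dark and bright colors through a panning water texture, adding curve-remapped specular highlights.
# Opacity erodes through two panning textures driven by dynamic parameters, shaped by a static switch between procedural bent mask and distorted drop noise.
# Curve atlas edge falloff, particle alpha and depth fade (skipped on mobile) finish the opacity.
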Material /Game/Effects/Fort_Effects/Events/CattusDoggus/Cattus/M_Cattus_WaterTrail
graph {
 properties {
  bUsedWithBeamTrails=true
  bUsedWithNanite=true
  bUsedWithNiagaraMeshParticles=true
  bUsedWithNiagaraRibbons=true
  bUsedWithParticleSprites=true
  blendMode=BLEND_Translucent
  pixelDepthOffsetMode=PDOM_Legacy
  twoSided=true
  usedShadingModels=Default Lit
 }
 nodes [
  n1 = VectorParameter() {group=WaterStuff, parameterName=DarkColor, sortPriority=1, defaultValue=<rgba 0.0151 0.0402 0.0497 1>}
  n2 = MaterialFunctionCall(UVs=TextureCoordinate(), "Rotation Angle (0-1)"=ScalarParameter() {parameterName=UV_Rotate}) {materialFunction=&CustomRotator}
  n3 = ComponentMask(None=VectorParameter() {group=WaterStuff, parameterName=WaterTexTile, sortPriority=3, defaultValue=<rgba 1.5 2 0 1>}.RGB) {g=true, r=true}
  n4 = TextureSampleParameter2D(UVs=Panner(
    Coordinate=Multiply(A=@n2.Rotated Values, B=@n3)
    Speed=VectorParameter() {group=WaterStuff, parameterName=Pan_WaterTex, sortPriority=4, defaultValue=<rgba -0.4 0.2 0 1>}.RGB
   ) {speedX=-0.4, speedY=0.2}) {
   group=WaterStuff
   parameterName=WaterTex
   samplerType=SAMPLERTYPE_Normal
   sortPriority=2
   texture=/Game/Packages/DS_Fortnite_TRV_Terrain/Textures/T_TRV_SwampWater_01_N
  }
  n5 = LinearInterpolate(
   A=@n1.RGB
   B=VectorParameter() {group=WaterStuff, parameterName=BrightColor, defaultValue=<rgba 0.1493 0.4631 0.5029 1>}.RGB
   Alpha=LinearInterpolate(Alpha=OneMinus(None=@n4.R))
  )
  n6 = ComponentMask(None=VectorParameter() {group=WaterStuff, parameterName=WaterSpecularTile, sortPriority=6, defaultValue=<rgba 2 3 0 1>}.RGB) {g=true, r=true}
  n7 = VectorParameter() {group=WaterStuff, parameterName=Pan_WaterSpecular, sortPriority=7, defaultValue=<rgba -0.5 0.1 0 1>}
  n8 = TextureSampleParameter2D(UVs=Panner(Coordinate=Multiply(A=@n2.Rotated Values, B=@n6), Speed=@n7.RGB) {speedX=-0.5, speedY=0.1}) {
   group=WaterStuff
   parameterName=WaterSpecularTex
   samplerType=SAMPLERTYPE_Normal
   sortPriority=5
   texture=/Game/Packages/DS_Fortnite_TRV_Terrain/Textures/T_TRV_HF_WaterNormal
  }
  n9 = CurveAtlasRowParameter(CurveTime=Clamp(None=@n8.R)) {
   atlas=&AthenaCurveAtlas01
   curve=&/Curve_CattusActor2
   defaultValue=116
  }
  n10 = LinearInterpolate(B=VectorParameter() {parameterName=SpecularColor, defaultValue=<rgba 1 1 1 1>}.RGB, Alpha=Multiply(A=@n9.B) {constB=0.7})
  MaterialFunctionCall(Emissive=Multiply(A=Add(A=@n5, B=@n10)) {constB=0.6}) {materialFunction=&MF_FX_EmissiveAdjustments}
  n11 = DynamicParameter() {paramNames ["Offset(0-1)", "Erode(1-0)", "ErodeThreshold(1-0)", Param4]}
  n12 = ScalarParameter() {group=ErodeStuff, parameterName="ErodeThresholdCustom(0-1,Temp)", sortPriority=6}
  n13 = Add(
   A=Multiply(A=StaticSwitchParameter(True=@n11."ErodeThreshold(1-0)", False=@n12) {defaultValue=true, group=Triggers, parameterName="ErodeThreshold(CustomCtrl,0,ParicleCtrl,1)"}) {constB=-5.8}
   B=Constant() {r=-2.2}
  )
  n14 = Add(A=@n13)
  n15 = DynamicParameter() {paramNames ["Offset(0-1)", "Erode(1-0)", "ErodeThreshold(1-0)", Param4]}
  n16 = Add(A=@n2.Rotated Values, B=@n15."Offset(0-1)")
  n17 = ComponentMask(None=VectorParameter() {group=ErodeStuff, parameterName=ErodeA_Tile, sortPriority=1, defaultValue=<rgba 0.8 1 0 1>}.RGB) {g=true, r=true}
  n18 = TextureSampleParameter2D(UVs=Panner(
    Coordinate=Multiply(A=@n16, B=@n17)
    Speed=VectorParameter() {group=ErodeStuff, parameterName=Pan_ErodeA, sortPriority=2, defaultValue=<rgba -0.4 0.1 0 1>}.RGB
   ) {speedX=-0.4, speedY=0.1}) {
   group=ErodeStuff
   parameterName=ErodeA_Texture
   samplerType=SAMPLERTYPE_Normal
   texture=&T_TRV_SwampWater_01_N
  }
  n19 = Add(A=@n2.Rotated Values, B=Add(A=@n15."Offset(0-1)", B=Constant() {r=0.2}))
  n20 = ComponentMask(None=VectorParameter() {group=ErodeStuff, parameterName=ErodeB_Tile, sortPriority=4, defaultValue=<rgba 0.85 0.7 0 1>}.RGB) {g=true, r=true}
  n21 = TextureSampleParameter2D(UVs=Panner(
    Coordinate=Multiply(A=@n19, B=@n20)
    Speed=VectorParameter() {group=ErodeStuff, parameterName=Pan_ErodeB, sortPriority=5, defaultValue=<rgba -0.5 0.1 0 1>}.RGB
   ) {speedX=-0.5, speedY=0.1}) {
   group=ErodeStuff
   parameterName=ErodeB_texture
   samplerType=SAMPLERTYPE_Normal
   sortPriority=3
   texture=&T_TRV_SwampWater_01_N
  }
  n22 = Power(Base=Multiply(A=OneMinus(None=@n18.R), B=OneMinus(None=@n21.G)))
  n23 = DynamicParameter() {paramNames ["Offset(0-1)", "Erode(1-0)", "ErodeThreshold(1-0)", Param4]}
  n24 = Panner(Coordinate=@n2.Rotated Values) {speedX=-0.5}
  n25 = Sine(None=Multiply(A=Time()) {constB=0.4})
  n26 = Add(A=Sine(None=ComponentMask(None=@n24) {r=true}) {period=0.5}, B=Add(A=Multiply(A=@n25) {constB=0.1}) {constB=1.5}) {constB=0.9}
  n27 = ComponentMask(None=@n2.Rotated Values) {r=true}
  n28 = AppendVector(
   A=LinearInterpolate(Alpha=OneMinus(None=@n27)) {constB=0.3}
   B=LinearInterpolate(Alpha=ComponentMask(None=@n2.Rotated Values) {g=true}) {constB=0.33}
  )
  n29 = Constant2Vector()
  n30 = CurveAtlasRowParameter(CurveTime=OneMinus(None=SphereMask(A=@n28, B=@n29) {attenuationRadius=0.5, hardnessPercent=1})) {
   atlas=&AthenaCurveAtlas01
   curve=&Curve_CattusActor2
   defaultValue=116
  }
  n31 = MaterialFunctionCall(Float2=Sine(None=Divide(A=@n2.Rotated Values))) {materialFunction=&BreakOutFloat2Components}
  n32 = CurveAtlasRowParameter(CurveTime=Multiply(A=@n31.R, B=@n31.R)) {
   atlas=&AthenaCurveAtlas01
   curve=&Curve_CattusActor2
   defaultValue=116
  }
  n33 = Panner(Coordinate=@n2.Rotated Values) {speedX=0.1}
  n34 = ComponentMask(None=Multiply(
    A=TextureSample(UVs=@n33) {samplerType=SAMPLERTYPE_Normal, texture=/Game/Global/Textures/Default/Noise/T_SwirlNoise_N}.RGB
    B=ScalarParameter() {defaultValue=1, parameterName=Distortion_Intensity}
   )) {g=true, r=true}
  n35 = Multiply(
   A=@n2.Rotated Values
   B=ComponentMask(None=VectorParameter() {group=Drop, parameterName=DropTile, sortPriority=1, defaultValue=<rgba 1 1 0 1>}.RGB) {g=true, r=true}
  )
  n36 = VectorParameter() {group=Drop, parameterName=Drop_Pan, sortPriority=2, defaultValue=<rgba 0 0 0 1>}
  n37 = TextureSampleParameter2D(UVs=Add(A=Multiply(A=@n34) {constB=0.1}, B=Panner(Coordinate=@n35, Speed=@n36.RGB))) {group=Drop, parameterName=DrobTexture, texture=/Game/Global/Textures/Default/Noise/TankMoteNoise}
  n38 = CurveAtlasRowParameter(CurveTime=ComponentMask(None=@n2.Rotated Values) {r=true}) {
   atlas=&AthenaCurveAtlas01
   curve=&Curve_CattusActor2
   defaultValue=116
  }
  n39 = ScalarParameter() {parameterName=Bottom_Enhance}
  n40 = StaticSwitchParameter(
   True=Multiply(A=Clamp(None=@n26), B=Multiply(A=@n30.R, B=@n32.G))
   False=Add(A=Multiply(A=@n37.R, B=@n32.G), B=LinearInterpolate(B=@n38.A, Alpha=@n39))
  ) {defaultValue=true, group=Triggers, parameterName=BentShap}
  n41 = Clamp(None=LinearInterpolate(A=LinearInterpolate(A=@n13, B=@n14, Alpha=@n22) {constA=-3, constB=-2}, B=Clamp(None=@n22), Alpha=Multiply(A=@n23."Erode(1-0)", B=@n40)) {constA=1, constB=2})
  n42 = DepthFade(Opacity=@n41, FadeDistance=ScalarParameter() {defaultValue=1000, parameterName=DepthFade_Distance}) {fadeDistanceDefault=1000}
  Multiply(A=ParticleColor().A, B=Multiply(A=ShadingPathSwitch(Default=@n42, Mobile=@n41) {desc=Migrated from FeatureLevelSwitch}, B=@n32.G))
  TextureCoordinate()
  TextureCoordinate()
  TextureCoordinate()
  TextureCoordinate()
  TextureCoordinate()
  TextureCoordinate()
  TextureCoordinate()
  TextureCoordinate()
 ]
}
